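# Preview physics asset for the large male FrenchFryXL retargeted outfit body, approximating it with one capsule per covered bone, including limb twist bones.
# Constraints link the bodies into a tree from the pelvis through the spine to head, arms and legs, with free angular motion on the three lower spine joints.
# Main thigh and left upper arm bones have no bodies, so calves constrain directly to the pelvis and the left forearm to the clavicle.
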PhysicsAsset /BRCosmetics/Characters/Player/Male/Large/Bodies/M_LRG_FrenchFryXL_Retarget/Physics/PA_Outfit_FrenchFryXL_Retarget_Preview
bodies {
 calf_l [{shapeType=Capsule, length=21.3306, radius=15.027, center=<-10.0638 -0.7046 1.6358>, rotation=<rot 0 -32.8338 -76.0938>}]
 calf_r [{shapeType=Capsule, length=19.5976, radius=15.9408, center=<9.855 1.0232 -0.6517>, rotation=<rot 0 -23.5227 -84.7924>}]
 calf_twist_02_l [{shapeType=Capsule, length=16.5237, radius=14.1075, center=<-7.275 0.4645 -0.2479>, rotation=<rot 0 174.4548 -75.0696>}]
 calf_twist_02_r [{shapeType=Capsule, length=16.709, radius=14.1956, center=<7.0737 -0.4444 0.2332>, rotation=<rot 0 -179.7562 -77.3246>}]
 clavicle_l [{shapeType=Capsule, length=21.0702, radius=13.3338, center=<28.9884 -1.1783 -2.4038>, rotation=<rot 26.8267 -90 -38.9343>}]
 foot_l [{shapeType=Capsule, length=22.6477, radius=14.567, center=<-1.8241 5.3401 0.9299>, rotation=<rot 0 177.1228 -77.4344>}]
 foot_r [{shapeType=Capsule, length=22.662, radius=14.5763, center=<2.4625 -5.3998 -0.9339>, rotation=<rot 0 177.6161 -77.4293>}]
 hand_l [{shapeType=Capsule, length=11.9863, radius=10.1894, center=<4.5911 -2.6423 -0.0271>, rotation=<rot 16.1845 -90 -53.3506>}]
 hand_r [{shapeType=Capsule, length=10.3598, radius=10.1355, center=<-5.9642 2.6214 1.0669>, rotation=<rot 15.5187 -90 -54.6845>}]
 head [{shapeType=Capsule, length=7.463, radius=22.5417, center=<0.1249 1.8422 0.4342>, rotation=<rot 0 -171.6713 -89.8472>}]
 lowerarm_l [{shapeType=Capsule, length=15.0974, radius=14.9005, center=<8.4583 -1.2561 1.7569>, rotation=<rot -33.7676 -90 -128.9943>}]
 lowerarm_r [{shapeType=Capsule, length=16.7082, radius=14.9918, center=<-7.024 1.3238 0.6128>, rotation=<rot 0 109.3587 -50.6545>}]
 pelvis [{shapeType=Capsule, length=23.7307, radius=36.0749, center=<-4.8761 1.8902 0.0897>, rotation=<rot 0 -164.2434 -177.6724>}]
 spine_01 [{shapeType=Capsule, length=30.06, radius=31.9165, shapeName=spine_01_capsule, center=<-2.7678 0 0>}]
 spine_02 [{shapeType=Capsule, length=32.0168, radius=20, shapeName=spine_02_capsule, center=<0 -14.2204 0>}]
 spine_03 [{shapeType=Capsule, length=31.6379, radius=20, shapeName=spine_03_capsule, center=<9.5694 -5.6739 0>}]
 spine_04 [{shapeType=Capsule, length=13.6361, radius=20, center=<13.2944 -2.3616 0>}]
 spine_05 [{shapeType=Capsule, length=13.6361, radius=20}]
 thigh_twist_01_l [{shapeType=Capsule, length=22.5383, radius=13.8822, center=<-2.4252 -3.7064 -8.019>, rotation=<rot 0 7.1468 -68.7714>}]
 thigh_twist_01_r [{shapeType=Capsule, length=21.459, radius=11.3353, center=<1.537 5.2475 10.1729>, rotation=<rot 0 4.1617 -62.376>}]
 thigh_twist_02_l [{shapeType=Capsule, length=18.5403, radius=16.2709, center=<-3.5617 0.982 0.5409>, rotation=<rot 0 -178.0771 -85.4983>}]
 thigh_twist_02_r [{shapeType=Capsule, length=19.1053, radius=17.2428, center=<2.8521 -1.6065 0.2427>, rotation=<rot 0 3.5749 -78.0012>}]
 upperarm_r [{shapeType=Capsule, length=14.9167, radius=11.0825, center=<-20.983 -1.0445 0.4078>, rotation=<rot 0 4.4821 -79.8095>}]
 upperarm_twist_01_l [{shapeType=Capsule, length=18.3644, radius=9.8367, center=<-3.2688 -4.9941 -0.3816>, rotation=<rot 0 -168.6795 -70.4328>}]
 upperarm_twist_01_r [{shapeType=Capsule, length=18.8713, radius=9.409, center=<2.6594 3.5452 -0.3828>, rotation=<rot 0 -173.5845 -78.7731>}]
 upperarm_twist_02_l [{shapeType=Capsule, length=15.3595, radius=11.3959, center=<0.2068 0.4267 -1.1969>, rotation=<rot 0 172.2381 -89.0448>}]
}
constraints [
 {bone1Name=spine_04, bone2Name=spine_03, swing1LimitDegrees=45, swing2LimitDegrees=45, twistLimitDegrees=45}
 {bone1Name=spine_05, bone2Name=spine_04, swing1LimitDegrees=45, swing2LimitDegrees=45, twistLimitDegrees=45}
 {bone1Name=clavicle_l, bone2Name=spine_05, swing1LimitDegrees=45, swing2LimitDegrees=45, twistLimitDegrees=45}
 {bone1Name=upperarm_twist_02_l, bone2Name=clavicle_l, swing1LimitDegrees=45, swing2LimitDegrees=45, twistLimitDegrees=45}
 {bone1Name=upperarm_twist_01_l, bone2Name=clavicle_l, swing1LimitDegrees=45, swing2LimitDegrees=45, twistLimitDegrees=45}
 {bone1Name=lowerarm_l, bone2Name=clavicle_l, swing1LimitDegrees=45, swing2LimitDegrees=45, twistLimitDegrees=45}
 {bone1Name=hand_l, bone2Name=lowerarm_l, swing1LimitDegrees=45, swing2LimitDegrees=45, twistLimitDegrees=45}
 {bone1Name=upperarm_r, bone2Name=spine_05, swing1LimitDegrees=45, swing2LimitDegrees=45, twistLimitDegrees=45}
 {bone1Name=upperarm_twist_01_r, bone2Name=upperarm_r, swing1LimitDegrees=45, swing2LimitDegrees=45, twistLimitDegrees=45}
 {bone1Name=lowerarm_r, bone2Name=upperarm_r, swing1LimitDegrees=45, swing2LimitDegrees=45, twistLimitDegrees=45}
 {bone1Name=hand_r, bone2Name=lowerarm_r, swing1LimitDegrees=45, swing2LimitDegrees=45, twistLimitDegrees=45}
 {bone1Name=head, bone2Name=spine_05, swing1LimitDegrees=45, swing2LimitDegrees=45, twistLimitDegrees=45}
 {bone1Name=calf_l, bone2Name=pelvis, swing1LimitDegrees=45, swing2LimitDegrees=45, twistLimitDegrees=45}
 {bone1Name=foot_l, bone2Name=calf_l, swing1LimitDegrees=45, swing2LimitDegrees=45, twistLimitDegrees=45}
 {bone1Name=calf_twist_02_l, bone2Name=calf_l, swing1LimitDegrees=45, swing2LimitDegrees=45, twistLimitDegrees=45}
 {bone1Name=thigh_twist_01_l, bone2Name=pelvis, swing1LimitDegrees=45, swing2LimitDegrees=45, twistLimitDegrees=45}
 {bone1Name=thigh_twist_02_l, bone2Name=pelvis, swing1LimitDegrees=45, swing2LimitDegrees=45, twistLimitDegrees=45}
 {bone1Name=calf_r, bone2Name=pelvis, swing1LimitDegrees=45, swing2LimitDegrees=45, twistLimitDegrees=45}
 {bone1Name=foot_r, bone2Name=calf_r, swing1LimitDegrees=45, swing2LimitDegrees=45, twistLimitDegrees=45}
 {bone1Name=calf_twist_02_r, bone2Name=calf_r, swing1LimitDegrees=45, swing2LimitDegrees=45, twistLimitDegrees=45}
 {bone1Name=thigh_twist_01_r, bone2Name=pelvis, swing1LimitDegrees=45, swing2LimitDegrees=45, twistLimitDegrees=45}
 {bone1Name=thigh_twist_02_r, bone2Name=pelvis, swing1LimitDegrees=45, swing2LimitDegrees=45, twistLimitDegrees=45}
 {
  bone1Name=spine_01
  bone2Name=pelvis
  swing1LimitDegrees=45
  swing1Motion=Free
  swing2LimitDegrees=45
  swing2Motion=Free
  twistLimitDegrees=45
  twistMotion=Free
 }
 {
  bone1Name=spine_02
  bone2Name=spine_01
  swing1LimitDegrees=45
  swing1Motion=Free
  swing2LimitDegrees=45
  swing2Motion=Free
  twistLimitDegrees=45
  twistMotion=Free
 }
 {
  bone1Name=spine_03
  bone2Name=spine_02
  swing1LimitDegrees=45
  swing1Motion=Free
  swing2LimitDegrees=45
  swing2Motion=Free
  twistLimitDegrees=45
  twistMotion=Free
 }
]
